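# Sound cue for the end of the Icy Shins freeze, judging by its name.
# A single wave player feeds a Modulator left at default settings, with the cue itself lowering pitch and volume.
# It borrows small melee weapon attenuation and the melee weapon sound class, and joins the shared Blaster global concurrency set.
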SoundCue /TeaCakeIcyShins/Assets/Audio/SC_IcyShins_FreezeEnd
root=Modulator_0
nodes [
 {id=Modulator_0, children [WavePlayer_1]}
 {id=WavePlayer_1, properties {soundWaveAssetPtr=&/IcyShins_FreezeEnd_01}}
]
properties {
 attenuationSettings=&Fort_Weapon_Melee_Sml_Att
 maxDistance=2800
 pitchMultiplier=0.9
 soundClassObject=&Fort_WeaponMelee_PSM
 volumeMultiplier=0.5
 concurrencyOverrides {resolutionRule=PreventNew}
 concurrencySet [/Game/Sounds/Concurrency_Rules/Blaster_Global_Concurrency]
}
