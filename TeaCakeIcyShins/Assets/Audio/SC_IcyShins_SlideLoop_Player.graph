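# Sound cue for the looping Icy Shins slide sound on a player, classed as third-person enemy foley.
# A looping wave player feeds a continuous modulator driven by the Speed parameter, which scales volume and shifts pitch slightly around normal.
# It plays at reduced volume with small melee-weapon attenuation, and its concurrency is limited per owner.
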SoundCue /TeaCakeIcyShins/Assets/Audio/SC_IcyShins_SlideLoop_Player
root=ModulatorContinuous_0
nodes [
 {
  id=ModulatorContinuous_0
  children [WavePlayer_0]
  properties {pitchModulationParams {maxOutput=1.1, minOutput=0.9, parameterName=Speed}, volumeModulationParams {parameterName=Speed}}
 }
 {id=WavePlayer_0, properties {bLooping=true, soundWaveAssetPtr=&/IcyShins_SlideLoop_Player_01}}
]
properties {
 attenuationSettings=&Fort_Weapon_Melee_Sml_Att
 maxDistance=2800
 soundClassObject=&Fort_Player_Foley_3P_Enemy_Athena
 volumeMultiplier=0.3
 concurrencyOverrides {bLimitToOwner=true}
}
