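# Translucent skeletal-mesh material for an ice block effect, built by running MF_IceShin between shared character PreFX and PostFX functions.
# Optional snow uses pre-skinned up-facing normals and projected Perlin noise, keeping it stable under animation.
# View-dependent glints from coordinate and direction textures, boosted by a sun lobe, alter emissive and optionally other surface attributes, weighted differently for ice and snow.
# Opacity passes through depth fade and translucency sorting.
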Material /TeaCakeIcyShins/Assets/FX/M_IceShin_IceBlock
graph {
 properties {
  allowTranslucentCustomDepthWrites=true
  bUseMaterialAttributes=true
  bUsedWithMeshDeformer=true
  bUsedWithSkeletalMesh=true
  blendMode=BLEND_Translucent
  pixelDepthOffsetMode=PDOM_Legacy
  shadingModel=MSM_FromMaterialExpression
  usedShadingModels=Default Lit | Subsurface | Clear Coat | Subsurface Profile | Cloth
 }
 nodes [
  n1 = StaticSwitchParameter(True=NamedRerouteUsage(), False=NamedRerouteUsage()) {parameterName=UseSnow}
  n2 = MaterialFunctionCall!(Opacity=DepthFade(
    Opacity=ScalarParameter() {defaultValue=0.9, parameterName=Opacity}
    FadeDistance=ScalarParameter() {defaultValue=2, parameterName=Depth Fade}
   )) {materialFunction=/Game/Global/Materials/MaterialGym/Materials/SortedOutline/MF_SortTranslucency}
  n3 = MaterialFunctionCall() {materialFunction=&MF_CharacterParent_2023_PreFX}
  MaterialFunctionCall(
   Material=SetMaterialAttributes!(MaterialAttributes=SetMaterialAttributes!(MaterialAttributes=@n1, Opacity=@n2.Result) {attributeSetTypes [#1]})
   M=@n3.M
   Pre Skinned Local Position=@n3.M
   Distance Blend=@n3.M
   Fresnel=@n3.M
  ) {materialFunction=&MF_CharacterParent_2023_PostFX}
  n4 = MaterialFunctionCall!(In=@n3.Result, InMask, InFXMask=@n3.Result) {materialFunction=&/MF_IceShin}
  NamedRerouteDeclaration(None=@n4.Result) {name=InMaterialAttributes, nodeColor=<rgba 1 0.3281 0 1>}
  n5 = NamedRerouteDeclaration(None=GetMaterialAttributes(None=NamedRerouteUsage()) {attributeGetTypes [#2, #3, #4]}.Normal) {name=InNormal, nodeColor=<rgba 1 0.5625 0 1>}
  n6 = Multiply(A=Fresnel(Normal=Transform(None=@n5)) {baseReflectFraction=0, exponent=1}) {constB=2}
  n7 = SetMaterialAttributes!(
   MaterialAttributes=NamedRerouteUsage()
   Base Color=LinearInterpolate(A=VectorParameter() {parameterName=SnowDarkColor, defaultValue=<rgba 0.1648 0.3186 0.3906 1>}.RGB, Alpha=Saturate(None=@n6)) {constA=1, constB=0.9}
   Emissive Color=Constant()
   Normal=NamedRerouteUsage()
   Roughness=Constant() {r=0.75}
   Specular=Constant() {r=0.8}
  ) {attributeSetTypes [
    #3
    #2
    #4
    #5
    #6
   ]}
  n8 = BlendMaterialAttributes(A=NamedRerouteUsage(), B=@n7, Alpha=NamedRerouteUsage())
  n9 = NamedRerouteUsage()
  n10 = VectorParameter() {group=GlitterFX V2, parameterName=GlintsBaseColor, sortPriority=10, defaultValue=<rgba 1 1 1 1>}
  n11 = Multiply(A=NamedRerouteUsage(), B=ScalarParameter() {group=GlitterFX V2, parameterName=GlintsBaseColor_Influence, sortPriority=10})
  n12 = NamedRerouteUsage()
  n13 = ScalarParameter() {defaultValue=0.2, group=GlitterFX V2, parameterName=GlintsMetallic, sortPriority=15}
  n14 = Multiply(A=NamedRerouteUsage(), B=ScalarParameter() {group=GlitterFX V2, parameterName=GlintsMetallic_Influence, sortPriority=15})
  n15 = NamedRerouteUsage()
  n16 = ScalarParameter() {defaultValue=0.2, group=GlitterFX V2, parameterName=GlintsRoughness, sortPriority=20}
  n17 = Multiply(A=NamedRerouteUsage(), B=ScalarParameter() {defaultValue=0.2, group=GlitterFX V2, parameterName=GlintsRoughness_Influence, sortPriority=20})
  n18 = NamedRerouteUsage()
  n19 = LinearInterpolate(
   A=Constant3Vector() {constant=<rgba 0 0 1 0>}
   B=Transform(None=NamedRerouteUsage()) {transformSourceType=TRANSFORMSOURCE_World, transformType=TRANSFORM_Tangent}
   Alpha=Multiply(A=NamedRerouteUsage(), B=ScalarParameter() {defaultValue=1, group=GlitterFX V2, parameterName=GlintsNormal_Influence, sortPriority=25})
  )
  n20 = SetMaterialAttributes!(
   MaterialAttributes=@n8
   Base Color=StaticSwitchParameter(True=LinearInterpolate(A=@n9, B=@n10.RGB, Alpha=@n11), False=@n9) {group=GlitterFX V2, parameterName=Glint_AffectBaseColor, sortPriority=9}
   Metallic=StaticSwitchParameter(True=LinearInterpolate(A=@n12, B=@n13, Alpha=@n14), False=@n12) {group=GlitterFX V2, parameterName=Glint_AffectMetallic, sortPriority=14}
   Roughness=StaticSwitchParameter(True=LinearInterpolate(A=@n15, B=@n16, Alpha=@n17), False=@n15) {group=GlitterFX V2, parameterName=Glint_AffectRoughness, sortPriority=19}
   Emissive Color=Add(A=NamedRerouteUsage(), B=NamedRerouteUsage())
   Normal=StaticSwitchParameter(
    True=MaterialFunctionCall(BaseNormal=@n18, AdditionalNormal=@n19) {materialFunction=&BlendAngleCorrectedNormals}.Result
    False=@n18
   ) {group=GlitterFX V2, parameterName=Glint_AffectNormal, sortPriority=24}
  ) {attributeSetTypes [
    #3
    #7
    #5
    #2
    #4
   ]}
  n21 = LinearInterpolate(
   A=ScalarParameter() {defaultValue=0.01, group=GlitterFX V2, parameterName=Glint - Ice Intensity}
   B=ScalarParameter() {defaultValue=1, group=GlitterFX V2, parameterName=Glint - Snow Intensity}
   Alpha=NamedRerouteUsage()
  )
  NamedRerouteDeclaration(None=StaticSwitchParameter(True=BlendMaterialAttributes(A=@n8, B=@n20, Alpha=@n21) {vertexAttributeBlendType=UseA}, False=@n8) {defaultValue=true, group=GlitterFX V2, parameterName=Enable Glints, sortPriority=-100}) {name=OutMaterialAttributes, nodeColor=<rgba 0.0078 0 1 1>}
  n22 = Multiply(
   A=Multiply(A=TextureCoordinate(), B=ScalarParameter() {defaultValue=1, group=GlitterFX V2, parameterName=Glint_Scale})
   B=ScalarParameter() {defaultValue=10, group=GlitterFX V2, parameterName=Glint_ScalePerMaterial}
  )
  n23 = TextureSampleParameter2D!(UVs=@n22, Bias=Constant() {r=-2}, Apply View MipBias) {
   group=GlitterFX V2
   mipValueMode=TMVM_MipBias
   parameterName=Glints_CoordsAndMask
   samplerType=SAMPLERTYPE_Masks
   texture=/TeaCakeIcyShins/Assets/FX/Textures/T_Glints_M
  }
  n24 = TextureSampleParameter2D!(UVs=ComponentMask(None=@n23.RGB) {g=true, r=true}, "DDX(UVs)"=DDX(Value=@n22), "DDY(UVs)"=DDY(Value=@n22), Apply View MipBias) {
   group=GlitterFX V2
   mipValueMode=TMVM_Derivative
   parameterName=Glints_WorldDirection
   samplerSource=SSM_Wrap_WorldGroupSettings
   samplerType=SAMPLERTYPE_LinearColor
   texture=/TeaCakeIcyShins/Assets/FX/Textures/T_Glints_N
  }
  n25 = ConstantBiasScale(None=@n24.RGB) {bias=-0.5, scale=2}
  n26 = Multiply(A=VertexNormalWS(), B=ScalarParameter() {group=GlitterFX V2, parameterName=GlintNormals_VertexNormalInfluence, sliderMax=1})
  n27 = NamedRerouteDeclaration(None=Normalize(VectorInput=Add(A=@n25, B=@n26))) {name=Glint Normal_1, nodeColor=<rgba 0 1 0.2734 1>}
  n28 = CameraVectorWS()
  n29 = Multiply(
   A=LinearInterpolate(
    A=Multiply(A=@n28) {constB=-1}
    B=MaterialFunctionCall() {materialFunction=&CameraDirectionVector}.Result
    Alpha=ScalarParameter() {group=GlitterFX V2, parameterName=GlintNormals_CameraVectorUniformity, sliderMax=1}
   )
   B=ScalarParameter() {defaultValue=2, group=GlitterFX V2, parameterName=GlintNormals_ScaleCameraVector}
  ) {constB=2}
  n30 = ScalarParameter() {defaultValue=15, group=GlitterFX V2, parameterName=Glint_NoiseFrequency}
  n31 = Frac(None=Multiply(A=DotProduct(A=@n27, B=@n29), B=@n30) {constB=7})
  n32 = Abs(None=Multiply(A=ConstantBiasScale(None=@n31) {bias=-0.5, scale=2}, B=Add(B=@n30) {constA=1}))
  n33 = LinearInterpolate(
   A=ScalarParameter() {defaultValue=0.25, group=GlitterFX V2, parameterName=Glint_NoiseThreshold, sliderMax=0.999}
   B=ScalarParameter() {defaultValue=0.25, group=GlitterFX V2, parameterName=Glint_NoiseThreshold_Lit, sliderMax=0.999}
   Alpha=NamedRerouteUsage()
  )
  n34 = Subtract(A=ConstantBiasScale(None=Sine(None=@n32)), B=@n33)
  n35 = OneMinus(None=@n33)
  n36 = Power(
   Base=Multiply(A=ScalarParameter() {defaultValue=1, parameterName=Glint Intensity}, B=Divide(A=@n34, B=@n35))
   Exp=ScalarParameter() {defaultValue=3, group=GlitterFX V2, parameterName=Glint_NoiseExponentBase}
  )
  n37 = SkyAtmosphereLightDirection()
  n38 = Add(A=@n37, B=CameraVectorWS())
  n39 = VertexNormalWS()
  n40 = DotProduct(A=@n39, B=@n37)
  n41 = Multiply(A=DotProduct(A=Normalize(VectorInput=@n38), B=@n39), B=Saturate(None=Add(A=@n40)))
  n42 = ScalarParameter() {defaultValue=1, group=GlitterFX V2, parameterName=SunDirectionLobe_Contrast}
  n43 = Power(
   Base=SmoothStep(Value=Multiply(A=@n41, B=@n42)) {constMin=-1}
   Exp=ScalarParameter() {defaultValue=4, group=GlitterFX V2, parameterName=SunDirectionLobeExp}
  )
  n44 = LinearInterpolate(
   A=ScalarParameter() {defaultValue=3, group=GlitterFX V2, parameterName=GlintExponent}
   B=ScalarParameter() {defaultValue=10, group=GlitterFX V2, parameterName=GlintExponentLit}
   Alpha=NamedRerouteDeclaration(None=Saturate(None=@n43)) {name=GlintsLitDirectionMask, nodeColor=<rgba 0 1 0.8125 1>}
  )
  n45 = NamedRerouteDeclaration(None=Multiply(A=Power(Base=@n36, Exp=@n44), B=NamedRerouteUsage())) {name=Glints MaskOutput, nodeColor=<rgba 1 0.9141 0 1>}
  n46 = LinearInterpolate(
   A=ScalarParameter() {defaultValue=10, group=GlitterFX V2, parameterName=GlintBrightness}
   B=ScalarParameter() {defaultValue=10, group=GlitterFX V2, parameterName=GlintBrightnessLit}
   Alpha=Saturate(None=NamedRerouteUsage())
  )
  NamedRerouteDeclaration(None=RayTracingQualitySwitch(Normal=Multiply(A=@n45, B=@n46) {constB=50}, RayTraced=@n45)) {name=Glints, nodeColor=<rgba 0.2891 1 0 1>}
  n47 = Multiply(
   A=Multiply(A=TextureCoordinate(), B=ScalarParameter() {defaultValue=1, group=GlitterFX V2, parameterName=Glint_Scale})
   B=ScalarParameter() {defaultValue=10, group=GlitterFX V2, parameterName=Glint_ScalePerMaterial}
  )
  n48 = TextureSampleParameter2D!(UVs=@n47, Bias=Constant() {r=-2}, Apply View MipBias) {
   group=GlitterFX V2
   mipValueMode=TMVM_MipBias
   parameterName=Glints_CoordsAndMask
   samplerType=SAMPLERTYPE_Masks
   texture=&T_Glints_M
  }
  n49 = TextureSampleParameter2D!(UVs=ComponentMask(None=@n48.RGB) {g=true, r=true}, "DDX(UVs)"=DDX(Value=@n47), "DDY(UVs)"=DDY(Value=@n47), Apply View MipBias) {
   group=GlitterFX V2
   mipValueMode=TMVM_Derivative
   parameterName=Glints_WorldDirection
   samplerSource=SSM_Wrap_WorldGroupSettings
   samplerType=SAMPLERTYPE_LinearColor
   texture=&T_Glints_N
  }
  n50 = ConstantBiasScale(None=@n49.RGB) {bias=-0.5, scale=2}
  n51 = Multiply(A=VertexNormalWS(), B=ScalarParameter() {group=GlitterFX V2, parameterName=GlintNormals_VertexNormalInfluence, sliderMax=1})
  NamedRerouteDeclaration(None=Normalize(VectorInput=Add(A=@n50, B=@n51))) {name=Glint Normal, nodeColor=<rgba 0 1 0.2734 1>}
  NamedRerouteDeclaration(None=@n48.A) {name=Glints_RawMask, nodeColor=<rgba 0.0312 0 1 1>}
  n52 = GetMaterialAttributes(None=@n8) {attributeGetTypes [
    #3
    #7
    #5
    #4
    #2
   ]}
  NamedRerouteDeclaration(None=@n52.Base Color) {name=BC, nodeColor=<rgba 1 0 0.6328 1>}
  NamedRerouteDeclaration(None=@n52.Metallic) {name=Metallic, nodeColor=<rgba 0 0.9531 1 1>}
  NamedRerouteDeclaration(None=@n52.Roughness) {name=Roughness, nodeColor=<rgba 1 0.7266 0 1>}
  NamedRerouteDeclaration(None=@n52.Normal) {name=Normal, nodeColor=<rgba 0.4062 0 1 1>}
  NamedRerouteDeclaration(None=@n52.Emissive Color) {name=Emissive, nodeColor=<rgba 0 1 0.0859 1>}
  NamedRerouteDeclaration(None=@n4.IceMask) {name=IceMask, nodeColor=<rgba 0 1 0.6484 1>}
  n53 = ScalarParameter() {defaultValue=0.9, parameterName=SnowHardness}
  n54 = Add(A=LocalPosition() {localOrigin=InstancePreSkinning}.XYZ, B=VectorParameter() {parameterName=NoiseOffset, defaultValue=<rgba 150 0 0 1>}.RGB)
  n55 = VertexInterpolator(VS=@n54)
  n56 = PreSkinnedNormal()
  n57 = MaterialFunctionCall(
   TextureObject=TextureObject() {texture=/TeaCakeIcyShins/Assets/FX/Textures/T_Perlin_Noise_M1}
   TextureSize=Constant() {r=200}
   WorldPosition=NamedRerouteDeclaration(None=ShaderStageSwitch(PixelShader=@n55.PS, VertexShader=@n54)) {name=PreSkinnedPosition, nodeColor=<rgba 1 0 0.8203 1>}
   World Space Normal=ShaderStageSwitch(PixelShader=VertexInterpolator(VS=@n56).PS, VertexShader=@n56)
  ) {materialFunction=&WorldAlignedTexture}
  n58 = PreSkinnedNormal()
  n59 = ShaderStageSwitch(PixelShader=VertexInterpolator(VS=@n58).PS, VertexShader=@n58)
  n60 = Add(A=ConstantBiasScale(None=ComponentMask(None=@n59) {b=true}), B=ScalarParameter() {defaultValue=0.01, parameterName=SnowHeight})
  n61 = ScalarParameter() {defaultValue=6, parameterName=SnowTransition}
  n62 = InverseLinearInterpolate(A=@n57.Z Texture, Value=Saturate(None=Power(Base=@n60, Exp=@n61)))
  n63 = ScalarParameter() {defaultValue=0.75, parameterName=SnowInfill}
  n64 = Multiply(
   A=Multiply(A=SmoothStep(Max=@n53, Value=@n62), B=LinearInterpolate(A=@n63, Alpha=@n57.Z Texture) {constA=0.25})
   B=ScalarParameter() {defaultValue=1, parameterName=SnowIntensity}
  )
  NamedRerouteDeclaration(None=Multiply(A=Saturate(None=@n64), B=NamedRerouteUsage())) {name=SnowMask, nodeColor=<rgba 0 0.1797 1 1>}
  NamedRerouteDeclaration(None=NamedRerouteUsage()) {name=OutNormal, nodeColor=<rgba 0.5 1 0 1>}
  NamedRerouteUsage()
 ]
}
