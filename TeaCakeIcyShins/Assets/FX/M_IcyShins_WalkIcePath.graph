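# Translucent unlit Niagara ribbon material for the IcyShins walking ice path.
# MF_IcyShins_IceDecal supplies world-position-based ice, dissolved by a particle-offset side mask and faded by tip/tail mask and particle alpha.
# Depth testing is off; opacity uses custom-depth fade, a manual scene-depth cutoff, and a dithered four-tap mask keeping water shading-model or custom-depth pixels.
# The platform switch zeroes opacity on Switch, Android and PC performance mode.
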Material /TeaCakeIcyShins/Assets/FX/M_IcyShins_WalkIcePath
graph {
 properties {
  bAutomaticallySetUsageInEditor=false
  bDisableDepthTest=true
  bUsedWithNiagaraRibbons=true
  blendMode=BLEND_Translucent
  previewMesh=/Engine/Transient.MaterialEditorMeshComponent_14
  shadingModel=MSM_Unlit
  twoSided=true
  usedShadingModels=Unlit
 }
 nodes [
  n1 = TextureCoordinate()
  n2 = VectorParameter() {group=Side, parameterName=Side_EdgeMask_Control, channelNames {a="", b=Power, g=Multiply, r=Offset}, defaultValue=<rgba 0 1.3 0.7 1>}
  n3 = MaterialFunctionCall(X=DynamicParameter() {defaultValue=<rgba 0 0 0 0>, paramNames [OffsetEdge,Param2,Param3,Param4]}.OffsetEdge) {materialFunction=&MakeFloat3}
  n4 = MaterialFunctionCall!(
   WorldPosition=WorldPosition().XYZ
   Dissolve=MaterialFunctionCall!(
    UorV=MaterialFunctionCall(Float2=@n1) {materialFunction=&BreakOutFloat2Components}.G
    EdgeControl=Add(A=@n2, B=@n3.Result)
   ) {materialFunction=/Game/Effects/Niagara/MasterMaterials/Functions/Common/FXCommon_EdgeMask_Simple}.Mask
  ) {materialFunction=&/MF_IcyShins_IceDecal}
  n5 = TextureCoordinate() {coordinateIndex=1}
  n6 = VectorParameter() {
   group=TipAndTail
   parameterName=TipAndTail_EdgeMask_InvertControl
   channelNames {a="", b=Power, g=Multiply, r=Offset}
   defaultValue=<rgba 1 5 1 1>
  }
  n7 = VectorParameter() {group=TipAndTail, parameterName=TipAndTail_EdgeMask_Control, channelNames {a="", b=Power, g=Multiply, r=Offset}, defaultValue=<rgba 1 5 1 1>}
  n8 = Multiply(
   A=Multiply(A=@n4.Opacity, B=ParticleColor().A)
   B=MaterialFunctionCall!(
    UorV=MaterialFunctionCall(Float2=@n5) {materialFunction=&BreakOutFloat2Components}.R
    Invert_EdgeControl=@n6
    NoInvert_EdgeControl=@n7
   ) {materialFunction=/Game/Effects/Niagara/MasterMaterials/Functions/Common/FXCommon_EdgeMask_Ex}.Mask
  )
  n9 = Subtract(A=SceneTexture() {sceneTextureId=PPI_CustomDepth}.Color, B=PixelDepth().R)
  n10 = ScalarParameter() {defaultValue=10, parameterName=DepthFade}
  n11 = ScreenPosition()
  n12 = CameraPositionWS()
  n13 = WorldPosition()
  n14 = ViewProperty() {property=MEVP_TemporalSampleIndex}
  MaterialExpressionAdd_8 = Add(A=ScreenPosition().PixelPosition, B=AppendVector(A=@n14.Property, B=@n14.Property))
  n15 = Custom!(p=@MaterialExpressionAdd_8) {
   code="Mod( ((uint)(p.x) + 2 * (uint)(p.y)) , 5 )"
   outputType=CMOT_Float1
   inputs [{inputName=p, input {expression=&self:MaterialExpressionAdd_8}}]
  }
  n16 = Multiply(
   A=Divide(A=ScalarParameter() {defaultValue=20, parameterName=F}, B=Distance(A=@n12, B=@n13.XYZ))
   B=LinearInterpolate(
    A=ScalarParameter() {defaultValue=0.1, parameterName=FLT}
    Alpha=Multiply(A=@n15) {constB=0.1666, desc=Divide by 6 avoiding a precision problem where an Alpha Threshold of 0 still leaves some remaining pixels.}
   )
  )
  n17 = MaterialFunctionCall!(OffsetAmount=MaterialFunctionCall(X=@n16, Y=@n16) {materialFunction=&MakeFloat2}.Result) {materialFunction=/Engine/Functions/Engine_MaterialFunctions01/ImageAdjustment/BlurSampleOffsets}
  n18 = Add(A=@n11.ViewportUV, B=@n17."0,1")
  n19 = MaterialFunctionCall(Float3=SceneTexture(UVs=@n18) {sceneTextureId=PPI_ShadingModelID}.Color) {materialFunction=&BreakOutFloat3Components}
  n20 = SceneTexture(UVs=@n18) {sceneTextureId=PPI_CustomDepth}
  n21 = Ceil(None=Saturate(None=Subtract(A=@n20.Color) {constB=999999}))
  n22 = Constant() {r=1}
  n23 = Add(A=@n11.ViewportUV, B=@n17."1,0")
  n24 = MaterialFunctionCall(Float3=SceneTexture(UVs=@n23) {sceneTextureId=PPI_ShadingModelID}.Color) {materialFunction=&BreakOutFloat3Components}
  n25 = SceneTexture(UVs=@n23) {sceneTextureId=PPI_CustomDepth}
  n26 = Ceil(None=Saturate(None=Subtract(A=@n25.Color) {constB=999999}))
  n27 = Constant() {r=1}
  n28 = Add(
   A=If(A=LinearInterpolate(B=@n19.R, Alpha=@n21) {constA=10}, "A > B"=@n22, "A == B"=Constant(), "A < B"=@n22) {constB=10}
   B=If(A=LinearInterpolate(B=@n24.R, Alpha=@n26) {constA=10}, "A > B"=@n27, "A == B"=Constant(), "A < B"=@n27) {constB=10}
  )
  n29 = Add(A=@n11.ViewportUV, B=@n17."0,-1")
  n30 = SceneTexture(UVs=@n29) {sceneTextureId=PPI_ShadingModelID}
  n31 = Saturate(None=Subtract(A=SceneTexture(UVs=@n29) {sceneTextureId=PPI_CustomDepth}.Color) {constB=999999})
  n32 = Constant() {r=1}
  n33 = If(
   A=LinearInterpolate(
    B=MaterialFunctionCall(Float3=@n30.Color) {materialFunction=&BreakOutFloat3Components}.R
    Alpha=Ceil(None=@n31)
   ) {constA=10}
   "A > B"=@n32
   "A == B"=Constant()
   "A < B"=@n32
  ) {constB=10}
  n34 = Add(A=@n11.ViewportUV, B=@n17."-1,0")
  n35 = Subtract(A=SceneTexture(UVs=@n34) {sceneTextureId=PPI_CustomDepth}.Color) {constB=999999}
  n36 = LinearInterpolate(
   B=MaterialFunctionCall(Float3=SceneTexture(UVs=@n34) {sceneTextureId=PPI_ShadingModelID}.Color) {materialFunction=&BreakOutFloat3Components}.R
   Alpha=Ceil(None=Saturate(None=@n35))
  ) {constA=10}
  n37 = Constant() {r=1}
  n38 = Constant()
  n39 = Divide(A=Add(A=Add(A=@n28, B=@n33), B=If(A=@n36, "A > B"=@n37, "A == B"=@n38, "A < B"=@n37) {constB=10})) {constB=4}
  n40 = ScalarParameter() {defaultValue=1, parameterName=WaterIDPower}
  n41 = LinearInterpolate(A=Clamp(None=Divide(A=@n9, B=@n10)), B=Constant(), Alpha=Saturate(None=Power(Base=@n39, Exp=@n40))) {constA=1, constAlpha=1}
  n42 = Constant()
  n43 = MaterialFunctionCall("Xb1,PS4"=@n41, "PC-High&&Epic,XSX,PS5"=@n41, "PC-Perf,Android"=@n42, Switch=@n42, "PC-Low(NotPerf)&&Medium"=@n41) {materialFunction=&PlatformSwitch_1D}
  n44 = Subtract(A=PixelDepth().R, B=SceneDepth().R)
  n45 = Constant() {r=300}
  n46 = Constant()
  n47 = Constant() {r=1}
  MakeMaterialAttributes!(
   BaseColor
   Metallic
   Specular
   Roughness
   Anisotropy
   EmissiveColor=MaterialFunctionCall(Emissive=Divide(A=@n4.Color)) {materialFunction=&MF_FX_EmissiveAdjustments}.Result
   Opacity=Multiply(A=Multiply(A=@n8, B=@n43.Result), B=If(A=@n44, B=@n45, "A > B"=@n46, "A == B"=@n47, "A < B"=@n47))
   OpacityMask
   Normal
   Tangent
   WorldPositionOffset
   SubsurfaceColor
   ClearCoat
   ClearCoatRoughness
   AmbientOcclusion
   Refraction
   CustomizedUVs_0
   CustomizedUVs_1
   CustomizedUVs_2
   CustomizedUVs_3
   CustomizedUVs_4
   CustomizedUVs_5
   CustomizedUVs_6
   CustomizedUVs_7
   PixelDepthOffset
   ShadingModel
   Displacement
  )
 ]
}
