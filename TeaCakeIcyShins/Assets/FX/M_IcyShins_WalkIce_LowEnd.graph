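# Unlit translucent, two-sided Niagara material for IcyShins walk ice, named as a low-end variant.
# An edge-faded UV sphere mask, merged with a particle-driven extra area whose center uses AdditionalX for both axes, feeds MF_IcyShins_IceDecal as Dissolve.
# A custom HLSL node in the emissive sum clips near-empty pixels.
# The BottomLayer switch substitutes vertex-color-blended tints for the decal color and adds vertex-color fading with temporal dither to opacity.
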Material /TeaCakeIcyShins/Assets/FX/M_IcyShins_WalkIce_LowEnd
graph {
 properties {
  bAutomaticallySetUsageInEditor=false
  bUseMaterialAttributes=true
  bUsedWithNiagaraMeshParticles=true
  bUsedWithNiagaraSprites=true
  bUsedWithWater=true
  blendMode=BLEND_Translucent
  previewMesh=/Engine/Transient.MaterialEditorMeshComponent_2
  shadingModel=MSM_Unlit
  twoSided=true
  usedShadingModels=Unlit
 }
 nodes [
  n1 = SphereMask(A=TextureCoordinate(), B=Constant() {r=0.5}) {attenuationRadius=1, hardnessPercent=0}
  n2 = VectorParameter() {group=EdgeFade, parameterName=EdgeFade_EdgeMask_Control, channelNames {a="", b=Power, g=Multiply, r=Offset}, defaultValue=<rgba 0 1.5 2 1>}
  n3 = Power(Base=Multiply(A=Add(A=@n1, B=@n2.Offset), B=@n2.Multiply), Exp=@n2.Power)
  n4 = DynamicParameter() {paramNames [Param1,AdditionalX,AdditionalY,Radius]}
  n5 = ScalarParameter() {parameterName=SphereMaskHardness}
  n6 = SphereMask(A=TextureCoordinate(), B=ScalarParameter() {defaultValue=0.5, parameterName=SphereMaskSize}, Radius=@n4.Param1, Hardness=@n5)
  n7 = TransformPosition(None=WorldPosition().XYZ) {transformSourceType=TRANSFORMPOSSOURCE_World}
  n8 = DynamicParameter() {paramNames [Param1,AdditionalX,AdditionalY,Radius]}
  n9 = Multiply(
   A=SphereMask(
    A=ComponentMask(None=@n7) {g=true, r=true}
    B=AppendVector(A=@n8.AdditionalX, B=@n8.AdditionalX)
    Radius=Multiply(A=@n4.Param1, B=@n8.Radius)
    Hardness=@n5
   )
   B=Multiply(
    A=DynamicParameter() {parameterIndex=1, paramNames [AdditionalAreaOpacity, Param2, Param3, Param4]}.AdditionalAreaOpacity
    B=ScalarParameter() {defaultValue=0.9, parameterName=AdditionalAreaStrength}
   )
  )
  n10 = MaterialFunctionCall!(WorldPosition=WorldPosition().XYZ, Dissolve=Multiply(A=Saturate(None=@n3), B=Max(A=@n6, B=@n9))) {materialFunction=&/MF_IcyShins_IceDecal}
  MaterialExpressionCeil_4 = Ceil(None=@n10.Opacity)
  n11 = Custom!(Opacity=@MaterialExpressionCeil_4) {
   code="float mask = 0;\r\nif (Opacity <= 0.01)\r\n{\r\nclip(-1);\r\n}\r\nreturn mask;"
   outputType=CMOT_Float1
   inputs [{inputName=Opacity, input {expression=&self:MaterialExpressionCeil_4}}]
  }
  n12 = VectorParameter() {parameterName=BottomLayerColor, defaultValue=<rgba 1 1 1 1>}
  n13 = VectorParameter() {parameterName=BottomLayerColorB, defaultValue=<rgba 1 1 1 1>}
  n14 = OneMinus(None=VertexColor().R)
  n15 = Divide(A=StaticSwitchParameter(True=LinearInterpolate(A=@n12.RGB, B=@n13.RGB, Alpha=@n14), False=@n10.Color) {parameterName=BottomLayer})
  n16 = Power(Base=@n14, Exp=ScalarParameter() {defaultValue=1, parameterName=VCPower})
  n17 = ScalarParameter() {defaultValue=1, parameterName=VCMul}
  n18 = Multiply(A=StaticSwitchParameter(True=Multiply(A=@n16, B=@n17), False=Constant() {r=1}) {parameterName=BottomLayer}, B=@n10.Opacity)
  n19 = Constant()
  n20 = LinearInterpolate(
   A=If(
    A=@n18
    "A > B"=MaterialFunctionCall(Alpha Threshold=@n18) {materialFunction=&DitherTemporalAA}.Result
    "A == B"=@n19
    "A < B"=@n19
   )
   Alpha=Power(Base=@n18, Exp=ScalarParameter() {defaultValue=2, parameterName=DitherEdge})
  )
  MakeMaterialAttributes!(
   BaseColor
   Metallic
   Specular
   Roughness
   Anisotropy
   EmissiveColor=Add(
    A=@n11
    B=MaterialFunctionCall(Emissive=@n15) {materialFunction=&MF_FX_EmissiveAdjustments}.Result
   )
   Opacity=Multiply(A=StaticSwitchParameter(True=@n20, False=@n18) {parameterName=BottomLayer}, B=ParticleColor().A)
   OpacityMask=MaterialFunctionCall(Alpha Threshold=@n10.Opacity) {materialFunction=&DitherTemporalAA}.Result
   Normal
   Tangent
   WorldPositionOffset
   SubsurfaceColor
   ClearCoat
   ClearCoatRoughness
   AmbientOcclusion
   Refraction
   CustomizedUVs_0
   CustomizedUVs_1
   CustomizedUVs_2
   CustomizedUVs_3
   CustomizedUVs_4
   CustomizedUVs_5
   CustomizedUVs_6
   CustomizedUVs_7
   PixelDepthOffset
   ShadingModel
   Displacement
  )
 ]
}
