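# NS_IceGauntlets_FrostyWisps is a Niagara system that, by its name, provides frosty wisp effects for ice gauntlets.
# It holds three emitters, two named as Empty and one as LW_Smoke_Ground, with no renderers or emitter modules listed.
# User parameters expose a color defaulting to translucent pale blue, a fade-out value and an effect radius.
# SystemState is the only listed module, and how the parameters drive the emitters is not shown.
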NiagaraSystem /TeaCakeIcyShins/Assets/FX/NS_IceGauntlets_FrostyWisps
dependencies {usedDataInterfaces [], usedDynamicInputs [], usedModules [&SystemState]}
emitters {Empty {rendererClasses [], renderers []}, Empty001 {rendererClasses [], renderers []}, LW_Smoke_Ground {rendererClasses [], renderers []}}
summary {
 systemName=NS_IceGauntlets_FrostyWisps
 emitters [{emitterName=Empty, rendererClasses []}, {emitterName=Empty001, rendererClasses []}, {emitterName=LW_Smoke_Ground, rendererClasses []}]
 userVariables [
  User.Color {type=LinearColor, description="", defaultValue {struct=LinearColor, value=<rgba 0.6798 0.7937 1 0.2>}}
  User.FadeOut {type=NiagaraFloat, defaultValue=5, description=""}
  User.FX Radius {type=NiagaraFloat, defaultValue=100, description=""}
 ]
}
